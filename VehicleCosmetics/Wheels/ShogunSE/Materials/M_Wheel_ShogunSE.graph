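# Masked master material for the Shogun SE cosmetic vehicle wheel, layering animated emissive effects over the shared MF_WheelParent_Simple base.
# Rim and tire RGBA masks, split by vertex color red, drive rebuilt Base Color and Emissive with custom colors and Fresnel glow.
# Emissive adds sine pulsing and a panning gradient, whose tiling changes through the LOD2345 static switch.
# The result passes through vehicle damage, then de-rez transition functions.
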Material /VehicleCosmetics/Wheels/ShogunSE/Materials/M_Wheel_ShogunSE
graph {
 properties {
  bAutomaticallySetUsageInEditor=false
  bUseMaterialAttributes=true
  bUsedWithMeshDeformer=true
  bUsedWithSkeletalMesh=true
  blendMode=BLEND_Masked
  pixelDepthOffsetMode=PDOM_Legacy
  previewMesh=/VehicleCosmetics/Wheels/ShogunSE/Meshes/SK_Wheel_ShogunSE
  usedShadingModels=Default Lit
 }
 nodes [
  n1 = MaterialFunctionCall() {materialFunction=&MF_WheelParent_Simple}
  n2 = GetMaterialAttributes(None=@n1.MaterialAttributes) {attributeGetTypes [#1]}
  n3 = Constant3Vector()
  n4 = LinearInterpolate(A=@n3, B=NamedRerouteUsage(), Alpha=NamedRerouteUsage()) {constB=0}
  n5 = LinearInterpolate(A=Multiply(A=Multiply(A=@n4) {constB=0.1}, B=NamedRerouteUsage()), Alpha=VertexColor().R) {constB=0}
  n6 = NamedRerouteUsage()
  n7 = Transform(None=NamedRerouteUsage())
  n8 = Multiply(
   A=Multiply(A=Fresnel(Normal=@n7) {baseReflectFraction=0, exponent=1}, B=NamedRerouteUsage())
   B=ScalarParameter() {defaultValue=1, group=5_Special Effects, parameterName=RimEmissiveIntensity}
  )
  n9 = NamedRerouteUsage()
  n10 = LinearInterpolate(A=LinearInterpolate(A=Multiply(A=@n6, B=@n8), B=@n3, Alpha=OneMinus(None=@n9)) {constB=0}, B=NamedRerouteUsage(), Alpha=NamedRerouteUsage())
  n11 = Constant3Vector() {constant=<rgba 1 1 1 0>}
  n12 = NamedRerouteUsage()
  n13 = NamedRerouteUsage()
  n14 = Multiply(A=NamedRerouteUsage(), B=ScalarParameter() {defaultValue=0.3, group=5_Special Effects, parameterName=PulsingIntensity, sortPriority=7})
  n15 = NamedRerouteUsage()
  n16 = Fresnel(Normal=Transform(None=NamedRerouteUsage())) {exponent=1}
  n17 = NamedRerouteUsage()
  n18 = Multiply(A=Multiply(A=@n15) {constB=3}, B=Multiply(A=NamedRerouteUsage(), B=Multiply(A=@n16, B=@n17)))
  n19 = Multiply(A=@n15) {constB=5}
  n20 = Add(A=Multiply(A=@n17, B=NamedRerouteUsage()), B=NamedRerouteUsage())
  n21 = Multiply(A=NamedRerouteUsage()) {constB=3}
  n22 = Multiply(A=NamedRerouteUsage(), B=ScalarParameter() {defaultValue=0.5, group=5_Special Effects, parameterName=PannerIntensity, sortPriority=10})
  n23 = LinearInterpolate(
   A=Add(A=LinearInterpolate(A=@n10, B=@n11, Alpha=@n12), B=Multiply(A=@n13, B=@n14))
   B=Add(A=LinearInterpolate(A=@n18, B=@n19, Alpha=@n20), B=Multiply(A=@n21, B=@n22))
   Alpha=VertexColor().R
  )
  MaterialFunctionCall(In=MaterialFunctionCall(Material=SetMaterialAttributes!(MaterialAttributes=@n2.MaterialAttributes, Base Color=@n5, Emissive Color=@n23) {attributeSetTypes [#2, #3]}) {materialFunction=&MF_BaseVehicle_Damage}.Result) {materialFunction=&MF_VehicleDeRez_Transition}
  n24 = VectorParameter() {group=5_Special Effects, parameterName=CustomColor2, sortPriority=0, defaultValue=<rgba 1 0.9192 0.0185 1>}
  NamedRerouteDeclaration(None=@n24.RGB) {name=AccentColor_output, nodeColor=<rgba 1 0 0.6328 1>}
  n25 = MaterialFunctionCall(Color=@n1.TireRGBAMask) {materialFunction=&SplitComponents}
  NamedRerouteDeclaration(None=@n25.G) {name=TireGMask, nodeColor=<rgba 1 0.4453 0 1>}
  NamedRerouteDeclaration(None=ComponentMask(None=@n1.TireRGBAMask) {a=true}) {name=TireAMask, nodeColor=<rgba 0 0.1328 1 1>}
  n26 = Time()
  n27 = ScalarParameter() {defaultValue=0.5, group=5_Special Effects, parameterName=PulsingSpeed, sortPriority=8}
  n28 = Add(A=Sine(None=Multiply(A=@n26, B=@n27) {constB=0.2})) {constB=2}
  n29 = NamedRerouteUsage()
  n30 = NamedRerouteUsage()
  NamedRerouteDeclaration(None=Multiply(A=Divide(A=@n28) {constB=3}, B=Add(A=@n29, B=@n30))) {name=PulsingFX, nodeColor=<rgba 0.6172 1 0 1>}
  n31 = StaticSwitchParameter(True=TextureCoordinate() {uTiling=3.8}, False=TextureCoordinate() {uTiling=1.5}) {group=5_Special Effects, parameterName=LOD2345}
  n32 = Multiply(A=Time(), B=ScalarParameter() {defaultValue=1, group=5_Special Effects, parameterName=PanningSpeed, sortPriority=9})
  n33 = MaterialFunctionCall(
   In=TextureSample(UVs=Panner(Coordinate=@n31, Time=@n32) {speedX=0.5}) {texture=/VehicleCosmetics/Wheels/_SharedTextures/FX/LinearGradient_Balanced_Vertical}.RGB
   Contrast=Constant() {r=0.1}
  ) {materialFunction=&CheapContrast}
  NamedRerouteDeclaration(None=Multiply(A=SmoothStep(Value=@n33.Result) {constMin=0.2}, B=NamedRerouteUsage())) {name=PannerFX, nodeColor=<rgba 1 0.2812 0 1>}
  n34 = VectorParameter() {group=5_Special Effects, parameterName=CustomColor1, sortPriority=1, defaultValue=<rgba 0.3231 0.0015 0.0194 1>}
  NamedRerouteDeclaration(None=@n34.RGB) {name=CustomColor_output, nodeColor=<rgba 0.2656 1 0 1>}
  NamedRerouteDeclaration(None=@n25.B) {name=TireBMask, nodeColor=<rgba 0.125 0 1 1>}
  n35 = MaterialFunctionCall(Color=@n1.RimRGBAMask) {materialFunction=&SplitComponents}
  NamedRerouteDeclaration(None=@n35.R) {name=RimRMask, nodeColor=<rgba 0.5234 1 0 1>}
  NamedRerouteDeclaration(None=@n25.R) {name=TireRMask, nodeColor=<rgba 1 0.2344 0 1>}
  n36 = VertexColor()
  NamedRerouteDeclaration(None=Multiply(A=@n2.Normal, B=OneMinus(None=@n36.R))) {name=RimNormal_output, nodeColor=<rgba 0 0.0859 1 1>}
  NamedRerouteDeclaration(None=Multiply(A=@n2.Normal, B=VertexColor().R)) {name=TireNormal_output, nodeColor=<rgba 0.4062 1 0 1>}
  NamedRerouteDeclaration(None=@n35.G) {name=RimGMask, nodeColor=<rgba 0.1953 1 0 1>}
  NamedRerouteDeclaration(None=@n35.B) {name=RimBMask, nodeColor=<rgba 1 0 0.5156 1>}
  VectorParameter() {group=5_Special Effects, parameterName=CustomColor, sortPriority=1, defaultValue=<rgba 0.3231 0.0015 0.0194 1>}
  VectorParameter() {group=5_Special Effects, parameterName=AccentColor, sortPriority=0, defaultValue=<rgba 1 0.9192 0.0185 1>}
 ]
}
