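# Translucent animated glow FX material for the Shogun SE vehicle cosmetic wheel's skeletal mesh.
# A vertically panned flow texture distorts UVs into a circular gradient, which is contrast-adjusted to blend MiscColor1 toward a desaturated version for scaled emissive.
# The same mask drives opacity, faded at both U edges and multiplied by vertex red.
# Attributes pass through MF_Vehicle_FNPostFX, while the FireColor parameter is left unconnected.
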Material /VehicleCosmetics/Wheels/ShogunSE/Materials/M_Wheel_ShogunSE_FX
graph {
 properties {
  bAutomaticallySetUsageInEditor=false
  bUseMaterialAttributes=true
  bUsedWithMeshDeformer=true
  bUsedWithSkeletalMesh=true
  blendMode=BLEND_Translucent
  pixelDepthOffsetMode=PDOM_Legacy
  previewMesh=/VehicleCosmetics/Wheels/ShogunSE/Meshes/SK_Wheel_ShogunSE
  usedShadingModels=Default Lit
 }
 nodes [
  n1 = Multiply(A=ScalarParameter() {defaultValue=1, parameterName=EmissiveIntensity}) {constB=10}
  n2 = VectorParameter() {group=0_SpecialEffects, parameterName=MiscColor1, sortPriority=0, defaultValue=<rgba 1 0 0 1>}
  n3 = Panner(Coordinate=TextureCoordinate() {uTiling=0.6, vTiling=0.32}, Time=Time()) {speedY=0.9}
  n4 = TextureCoordinate()
  n5 = Constant2Vector() {g=-1.75, r=-0.5}
  n6 = Add(
   A=ComponentMask(None=TextureSample(UVs=@n3) {samplerType=SAMPLERTYPE_LinearColor, texture=/VehicleCosmetics/Wheels/ShogunSE/Textures/Flow_04}.RGB) {g=true, r=true}
   B=Multiply(A=Add(A=@n4, B=@n5)) {constB=0.17}
  )
  n7 = MaterialFunctionCall(
   In=Subtract(A=TextureSample(UVs=@n6) {samplerType=SAMPLERTYPE_Grayscale, texture=/VehicleCosmetics/Wheels/_SharedTextures/FX/GradientCircle01}.G) {constB=-0.2}
   Contrast=Constant() {r=0.5}
  ) {materialFunction=&CheapContrast}
  n8 = LinearInterpolate(A=@n2.RGB, B=Desaturation(None=@n2.RGB, Fraction=Constant() {r=0.2}), Alpha=@n7.Result)
  n9 = TextureCoordinate()
  n10 = ComponentMask(None=@n9) {r=true}
  n11 = Constant() {r=1.5}
  n12 = OneMinus(None=ComponentMask(None=@n9) {r=true})
  n13 = Multiply(
   A=MaterialFunctionCall(In=@n7.Result, Contrast=Constant() {r=0.5}) {materialFunction=&CheapContrast}.Result
   B=Multiply(A=Multiply(A=@n10, B=@n11) {constB=3}, B=Multiply(A=@n12, B=@n11) {constB=3})
  )
  n14 = VertexColor()
  MaterialFunctionCall!(In=MakeMaterialAttributes!(
    BaseColor
    Metallic
    Specular
    Roughness
    Anisotropy
    EmissiveColor=Multiply(A=@n1, B=@n8)
    Opacity=Multiply(A=@n13, B=@n14.R)
    OpacityMask
    Normal
    Tangent
    WorldPositionOffset
    SubsurfaceColor
    ClearCoat
    ClearCoatRoughness
    AmbientOcclusion
    Refraction
    CustomizedUVs_0
    CustomizedUVs_1
    CustomizedUVs_2
    CustomizedUVs_3
    CustomizedUVs_4
    CustomizedUVs_5
    CustomizedUVs_6
    CustomizedUVs_7
    PixelDepthOffset
    ShadingModel
    Displacement
   )) {materialFunction=/VehicleCosmetics/Materials/MaterialFunctions/MF_Vehicle_FNPostFX}
  Multiply(A=@n7.Result, B=@n7.Result)
  VectorParameter() {group=0_SpecialEffects, parameterName=FireColor, sortPriority=0, defaultValue=<rgba 1 0 0 1>}
 ]
}
